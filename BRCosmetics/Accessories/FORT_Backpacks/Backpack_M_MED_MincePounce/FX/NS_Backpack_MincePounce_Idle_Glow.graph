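# Idle glow effect for the MincePounce backpack cosmetic, drawing soft-fading sprites paired with lights.
# User.Variant gates burst spawn probability to choose between two emitters.
# Glow_Default cycles one particle from blue through magenta on engine time, while Glow_Variant partitions two offset particles into fixed cyan and magenta.
# Alpha fades in with emitter age and scales by quality level, time of day and front-end versus gameplay context.
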NiagaraSystem /BRCosmetics/Accessories/FORT_Backpacks/Backpack_M_MED_MincePounce/FX/NS_Backpack_MincePounce_Idle_Glow
dependencies {
 usedDataInterfaces [
  &NiagaraDataInterfaceColorCurve
  &NiagaraDataInterfaceCurve
  &NiagaraDataInterfaceArrayFloat
  &NiagaraDataInterfaceSocketReader
 ]
 usedDynamicInputs []
 usedModules [
  &self:SystemSpawnScript.SystemScriptSource.SystemScriptGraph.NiagaraNodeAssignment_0.SetVariables_#1
  &SystemState
 ]
}
emitters {
 Glow_Default {
  emitterUpdateScript [
   &EmitterState(
    Inactive Response=::NewEnumerator1
    Life Cycle Mode=::NewEnumerator1
    Loop Behavior=::NewEnumerator1
    Loop Duration Mode=::NewEnumerator1
    UseLoopDelay=false
   )
   &SpawnBurst_Instantaneous(
    Spawn Probability=&MakeCustomFloatFromBool(
     Bool=&SetBoolByFloatComparison(A=$User.Variant, B=0, Comparison Type=::NewEnumerator4)
     True Float=1
     False Float=0
    )
    Use Spawn Probability=true
    UseLoopCountLimit=false
   )
  ]
  particleSpawnScript [
   &InitializeParticle(
    Color Mode=::NewEnumerator0
    Lifetime=6
    Position Offset=&FN_FrontEndVector3Modifier(Gameplay=<0 -11.5 -4.5>, Front End=<0 -12 -4.5>)
    Sprite Size=&FN_FrontEndVector2Modifier(Front End:Vector2f=<15 23>, Gameplay:Vector2f=<17 23>)
    Sprite Size Mode=::NewEnumerator4
    UsePositionOffset=true
    Write Material Random=true
   )
   &CameraOffset(Camera Offset Amount=2)
  ]
  particleUpdateScript [
   &ParticleState(
    Kill Particles When Lifetime Has Elapsed=false
    Let Infinitely Lived Particles Die When Emitter Deactivates=true
    Loop Particles Lifetime=true
   )
   &SpriteFacingAndAlignment(Sprite Alignment=<0 -0.0587 1>, WriteAlignment=true, WriteFacing=false)
   &ScaleColor(
    Scale Alpha=&Multiply_Float(
     A=&FloatFromCurve(
      FloatCurve=DataInterface() {propertyValues {Curve {defaultValue=1, keys [{tangentMode=RCTM_User, time=0, value=0}, {tangentMode=RCTM_User, time=0.5, value=1}]}}}
      CurveIndex=$Emitter.Age
      Scale Curve=&FN_FrontEndFloatModifier(Gameplay=1, Front End=0.8)
     )
     B=&Multiply_Float(
      A=&SelectFloatFromArray(Direct Array Index=$Engine.QualityLevel, Float Selection Array=DataInterface() {propertyValues {FloatData [0.6,0.75,1,1,1]}})
      B=&FNTimeOfDayScalar(Night Scale=0.7, Morning Scale=0.8, Day Scale=0.8, Evening Scale=0.8)
     )
    )
    Scale RGB=&MakeVectorFromLinearColorRGB(Color=&ColorFromCurve(
      ColorCurve=DataInterface() {propertyValues {
        AlphaCurve {keys []}
        BlueCurve {keys [<at 0.0054 1>,<at 0.1321 1>,<at 0.2698 1>,<at 0.3866 1 RCIM_Cubic>,<at 0.5138 1>,<at 0.6838 1>,<at 0.8723 1>]}
        GreenCurve {keys [
          <at 0.0054 0.3372>
          <at 0.1321 0.0048>
          <at 0.2698 0.006>
          {arriveTangent=-0.0029, interpMode=RCIM_Cubic, leaveTangent=-0.0029, time=0.3866, value=0.007}
          <at 0.5138 -0.0149>
          <at 0.6838 0.0048>
          <at 0.8723 0.3372>
         ]}
        RedCurve {keys [
          <at 0.0054 0>
          <at 0.1321 0.0203>
          <at 0.2698 0.5502>
          {arriveTangent=-1.2799, interpMode=RCIM_Cubic, leaveTangent=-1.2799, time=0.3866, value=1}
          <at 0.5138 0.6717>
          <at 0.6838 0.0203>
          <at 0.8723 0>
         ]}
       }}
      CurveIndex=&ClampFloat(Float=&Frac_Float(Float=&Multiply_Float(A=$Engine.Time, B=0.2)))
     ))
    ScaleA=true
    ScaleRGB=true
   )
   &FN_Light_Attributes(
    Color=$Particles.Color
    Create Color Attribute=true
    Create Position Attribute=true
    Create Radius Attribute=true
    Position=&AddVectorToPosition(Position=$Particles.Position, Vector=<0 -6 1>)
    Radius=12
    Scale Color=&FN_FrontEndFloatModifier(Gameplay=7.5, Front End=7)
   )
   &DynamicMaterialParameters(
    Index 0 Param 1=0.9
    Index 0 Param 2=30
    Index 0 Param 3=1
    Module.Index0Float0Write=true
    Module.Index0Float1Write=true
    Module.Index0Float2Write=true
    Module.Index0Float3Write=true
   )
  ]
  rendererClasses [&NiagaraSpriteRendererProperties, &NiagaraLightRendererProperties]
  renderers [
   {rendererClass=&NiagaraSpriteRendererProperties}
   {rendererClass=&NiagaraLightRendererProperties, rendererIndex=1}
  ]
 }
 Glow_Variant {
  emitterUpdateScript [
   &EmitterState(
    Inactive Response=::NewEnumerator1
    Life Cycle Mode=::NewEnumerator1
    Loop Behavior=::NewEnumerator1
    Loop Duration Mode=::NewEnumerator1
    UseLoopDelay=false
   )
   &SpawnBurst_Instantaneous(
    Spawn Count=2
    Spawn Probability=&MakeCustomFloatFromBool(
     Bool=&SetBoolByFloatComparison(A=$User.Variant, B=1, Comparison Type=::NewEnumerator4)
     True Float=1
     False Float=0
    )
    Use Spawn Probability=true
    UseLoopCountLimit=false
   )
  ]
  particleSpawnScript [
   &PartitionParticles()
   &InitializeParticle(
    Color Mode=::NewEnumerator0
    Lifetime=6
    Position Offset=&MergeHelper_VectorSwitch(0=<0 -12 -1.5>, 1=<0 -12 -5.5>, 2=?, 3=?, Index=$Particles.PartitionParticles.Partition, Num=::NewEnumerator0)
    Sprite Size=&MergeHelper_Vector2Switch(0=<15 18.5>, 1=<15 18.5>, 2=?, 3=?, 4=?, Index=$Particles.PartitionParticles.Partition, Num=::NewEnumerator0)
    Sprite Size Mode=::NewEnumerator4
    UsePositionOffset=true
    Write Material Random=true
   )
   &CameraOffset()
  ]
  particleUpdateScript [
   &ParticleState(
    Kill Particles When Lifetime Has Elapsed=false
    Let Infinitely Lived Particles Die When Emitter Deactivates=true
    Loop Particles Lifetime=true
   )
   &SpriteFacingAndAlignment(Sprite Alignment=<0 -0.0587 1>, WriteAlignment=true, WriteFacing=false)
   &ScaleColor(
    Scale Alpha=&Multiply_Float(
     A=&FloatFromCurve(
      FloatCurve=DataInterface() {propertyValues {Curve {defaultValue=1, keys [{tangentMode=RCTM_User, time=0, value=0}, {tangentMode=RCTM_User, time=0.5, value=1}]}}}
      CurveIndex=$Emitter.Age
      Scale Curve=&FN_FrontEndFloatModifier(Gameplay=0.35, Front End=0.4)
     )
     B=&Multiply_Float(
      A=&SelectFloatFromArray(Direct Array Index=$Engine.QualityLevel, Float Selection Array=DataInterface() {propertyValues {FloatData [0.6,0.75,1,1,1]}})
      B=&FNTimeOfDayScalar(Night Scale=0.7, Morning Scale=0.8, Evening Scale=0.8)
     )
    )
    Scale RGB=&Lerp_Vector(
     A=&MakeVectorFromLinearColorRGB(Color=&ColorFromCurve(ColorCurve=DataInterface() {propertyValues {
         AlphaCurve {keys [<at 0 1 RCIM_Cubic>,<at 1 0 RCIM_Cubic>]}
         BlueCurve {keys [<at 0.3759 2.25 RCIM_Cubic>]}
         GreenCurve {keys [<at 0.3759 1.6304 RCIM_Cubic>]}
         RedCurve {keys [<at 0.3759 0 RCIM_Cubic>]}
        }}))
     B=&MakeVectorFromLinearColorRGB(Color=&ColorFromCurve(ColorCurve=DataInterface() {propertyValues {
         AlphaCurve {keys [<at 0 1 RCIM_Cubic>,<at 1 0 RCIM_Cubic>]}
         BlueCurve {keys [<at 0.3759 1.1942 RCIM_Cubic>]}
         GreenCurve {keys [<at 0.3759 0 RCIM_Cubic>]}
         RedCurve {keys [<at 0.3759 2.25 RCIM_Cubic>]}
        }}))
     Alpha=$Particles.PartitionParticles.NormalizedPartition
    )
    ScaleA=true
    ScaleRGB=true
   )
   &FN_Light_Attributes(
    Color=$Particles.Color
    Create Color Attribute=true
    Create Position Attribute=true
    Create Radius Attribute=true
    Position=&AddVectorToPosition(
     Position=$Particles.Position
     Vector=&MergeHelper_VectorSwitch(0=<0 -6 0>, 1=<0 -6 -2>, 2=?, 3=?, Index=$Particles.PartitionParticles.Partition, Num=::NewEnumerator0)
    )
    Radius=11
    Scale Color=&FN_FrontEndFloatModifier(Gameplay=3, Front End=5)
   )
   &DynamicMaterialParameters(
    Index 0 Param 1=0.9
    Index 0 Param 2=30
    Index 0 Param 3=1
    Module.Index0Float0Write=true
    Module.Index0Float1Write=true
    Module.Index0Float2Write=true
    Module.Index0Float3Write=true
   )
  ]
  rendererClasses [&NiagaraSpriteRendererProperties, &NiagaraLightRendererProperties]
  renderers [
   {rendererClass=&NiagaraSpriteRendererProperties}
   {rendererClass=&NiagaraLightRendererProperties, rendererIndex=1}
  ]
 }
}
summary {
 systemName=NS_Backpack_MincePounce_Idle_Glow
 emitters [
  {emitterName=Glow_Default, rendererClasses [&NiagaraSpriteRendererProperties, &NiagaraLightRendererProperties]}
  {emitterName=Glow_Variant, rendererClasses [&NiagaraSpriteRendererProperties, &NiagaraLightRendererProperties]}
 ]
 userVariables [
  User.bIsFrontEnd {type=NiagaraBool}
  User.bIsFrontEndPreview {type=NiagaraBool}
  User.Variant {type=NiagaraFloat, defaultValue=0, description=""}
 ]
}
